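; Anim notify for the Emperor emote that spawns an effect matched to the character's material variant.
; On Received_Notify it spawns the FX Niagara system attached to the mesh's root and keeps the component.
; It then casts the mesh owner to a Fort player pawn and reads the active Mat1/Mat2 index on the character's material variant channel.
; The second variant (index 1) sets the effect's bIsBodyDark parameter.
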(fn Received_Notify
  (params (MeshComp :type Skeletal%20Mesh%20Component%20Object%20Reference) (Animation :type Anim%20Sequence%20Base%20Object%20Reference) (EventReference :type Anim%20Notify%20Event%20Reference%20Structure%20%28by%20ref%29))
  (bind _returnvalue (Niagara|SpawnSystemAttached (Variables|Default|GetFX) MeshComp "root" "0, 0, 0" "0, 0, 0" "SnapToTarget" true))
  (bind _output_get (LocalVariables|SetNiagaraComp _returnvalue))
  (Utilities|FlowControl|Sequence
    (:then_2
      ;[1 0 0.5669 0.198] Check body color
      (bind _asfort_player_pawn (Utilities|Casting|CastToFortPlayerPawn (Components|GetOwner (Variables|GetMeshComp))))
      (bind _index (CosmeticLoadout|GetVariantChannelActiveIndexforSlotActor "(TagName=\"LoadoutSlot.Character\")" _asfort_player_pawn "(TagName=\"Cosmetics.Variant.Channel.Material\")" (Utilities|Array|MakeArray "(TagName=\"Cosmetics.Variant.Property.Mat1\")" "(TagName=\"Cosmetics.Variant.Property.Mat2\")")))
      (Effects|Components|ParticleSystem|SetBoolParameter (LocalVariables|GetNiagaraComp) "bIsBodyDark" (== _index 1)))))
