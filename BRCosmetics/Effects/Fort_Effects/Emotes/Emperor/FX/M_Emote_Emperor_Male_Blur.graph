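# Translucent unlit Niagara mesh-particle material that makes a soft, blurred scene distortion for the male Emperor emote effect.
# Emissive is a spiral blur of scene texture.
# A body-shape curve atlas sampled along V pushes vertices along normals and, via its slope, bends normals for an inverted Fresnel edge fade.
# Opacity combines that fade with top/bottom bounding-box masks, depth fade and particle alpha; the mobile path switches to a constant.
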Material /BRCosmetics/Effects/Fort_Effects/Emotes/Emperor/FX/M_Emote_Emperor_Male_Blur
graph {
 properties {
  bUsedWithNiagaraMeshParticles=true
  blendMode=BLEND_Translucent
  pixelDepthOffsetMode=PDOM_Legacy
  shadingModel=MSM_Unlit
  translucencyPass=MTP_BeforeDOF
  usedShadingModels=Unlit
 }
 nodes [
  n1 = Constant()
  n2 = MaterialFunctionCall!(
   Distance=ScalarParameter() {defaultValue=0.05, parameterName="Distance (S)"}
   Distance Steps=ScalarParameter() {defaultValue=16, parameterName="Distance Steps (S)"}
   Radial Steps=ScalarParameter() {defaultValue=8, parameterName="Radial Steps (S)"}
   Radial Offset=ScalarParameter() {defaultValue=0.62, parameterName="Radial Offset (S)"}
   TempAA Radial Blur=ScalarParameter() {defaultValue=1, parameterName="TempAA Radial Blur (S)"}
   TempAA Distance Blur=ScalarParameter() {defaultValue=1, parameterName="TempAA Distance Blur (S)"}
   Distance Mask=ScalarParameter() {defaultValue=1, parameterName="Distance Mask (S)"}
  ) {materialFunction=/Engine/Functions/Engine_MaterialFunctions01/Texturing/SpiralBlur-SceneTexture}
  n3 = Multiply(A=@n2.Result, B=ScalarParameter() {parameterName=SceneColorScale})
  MaterialFunctionCall(Emissive=ShadingPathSwitch(Default=@n3, Deferred=@n3, Mobile=@n1) {desc=Migrated from FeatureLevelSwitch}) {materialFunction=&MF_FX_EmissiveAdjustments}
  n4 = ScalarParameter() {defaultValue=1, parameterName=Fres_RegularFresExp}
  n5 = ScalarParameter() {defaultValue=22, parameterName=Fres_TopFresExp}
  n6 = Transform(None=Constant3Vector() {constant=<rgba 0 0 1 0>}) {transformSourceType=TRANSFORMSOURCE_Local}
  n7 = DotProduct(A=CameraVectorWS(), B=@n6)
  n8 = Saturate(None=Power(Base=Abs(None=@n7), Exp=ScalarParameter() {defaultValue=3, parameterName=Fres_NormalchangeClosePower}))
  n9 = PixelNormalWS()
  n10 = Constant3Vector() {constant=<rgba 0 0 -1 0>}
  n11 = ScalarParameter() {defaultValue=1.26, parameterName=Fres_ScaleNormalChange}
  n12 = MaterialFunctionCall(Float2=TextureCoordinate()) {materialFunction=&BreakOutFloat2Components}
  n13 = MaterialFunctionCall(X=@n12.G, Y=@n12.G) {materialFunction=&MakeFloat2}
  n14 = TextureSampleParameter2D(UVs=@n13.Result) {
   parameterName=BodyShapeTexture
   samplerType=SAMPLERTYPE_LinearColor
   texture=/BRCosmetics/Effects/Fort_Effects/Emotes/Emperor/FX/Texture/CurveAtlas_Emote_Emperor_Male_BodyWPO
  }
  n15 = TextureSampleParameter2D(UVs=Add(A=@n13.Result, B=ScalarParameter() {defaultValue=0.045, parameterName=Fres_StepHeight}) {constB=0.01}) {
   parameterName=BodyShapeTexture
   samplerType=SAMPLERTYPE_LinearColor
   texture=&CurveAtlas_Emote_Emperor_Male_BodyWPO
  }
  n16 = MaterialFunctionCall(
   1=Multiply(A=Transform(None=@n10) {transformSourceType=TRANSFORMSOURCE_Local}, B=@n11)
   2=@n9
   3=Multiply(A=@n6, B=@n11)
   A=Add(A=Subtract(A=@n14.R, B=@n15.R)) {constB=0.5}
  ) {materialFunction=&Lerp_Multiple_Float3}
  n17 = OneMinus(None=Fresnel(
    ExponentIn=LinearInterpolate(A=@n4, B=@n5, Alpha=@n8)
    BaseReflectFractionIn=Constant()
    Normal=LinearInterpolate(A=@n9, B=@n16.Lerp 3 Inputs, Alpha=@n8)
   ))
  n18 = ScalarParameter() {parameterName=Fres_FinalPower}
  n19 = Multiply(A=Multiply(A=@n2.SceneColor clamp to 0, B=Power(Base=@n17, Exp=@n18)), B=ScalarParameter() {parameterName=OpacityScale})
  n20 = MaterialFunctionCall() {materialFunction=&BoundingBoxBased_0-1_UVW}
  n21 = Add(A=OneMinus(None=@n20.B), B=ScalarParameter() {parameterName=BoundMask_TopOffset})
  n22 = ScalarParameter() {parameterName=BoundMaskSharpness}
  n23 = Add(A=@n20.B, B=ScalarParameter() {parameterName=BoundMask_BtmOffset})
  n24 = Saturate(None=Min(A=Multiply(A=@n21, B=@n22), B=Multiply(A=@n23, B=@n22)))
  n25 = Multiply(A=ParticleColor().A, B=DepthFade(Opacity=Multiply(A=@n19, B=@n24), FadeDistance=ScalarParameter() {parameterName=FadeDistance}))
  ShadingPathSwitch(Default=@n25, Deferred=@n25, Mobile=@n1) {desc=Migrated from FeatureLevelSwitch}
  n26 = Multiply(A=Multiply(A=VertexNormalWS(), B=@n14.R), B=ScalarParameter() {parameterName=WPOStr})
  ShadingPathSwitch(Default=@n26, Deferred=@n26, Mobile=@n1) {desc=Migrated from FeatureLevelSwitch}
 ]
}
