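# Translucent Niagara mesh-particle material for the male body effect in the Emperor emote.
# Curve atlases indexed by mesh V drive normal-direction vertex offset and two color/alpha gradients blended by CurveBlend, tinted by particle color.
# A panning texture mask, mapped by angle and height from local position, combines with top/bottom fades, inverted fresnel and depth fade for opacity.
# Fresnel normals bend by the shape curve's slope when viewed along local Z.
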Material /BRCosmetics/Effects/Fort_Effects/Emotes/Emperor/FX/M_Emote_Emperor_Male_Color
graph {
 properties {
  bUsedWithNiagaraMeshParticles=true
  blendMode=BLEND_Translucent
  pixelDepthOffsetMode=PDOM_Legacy
  translucencyLightingMode=TLM_VolumetricDirectional
  usedShadingModels=Default Lit
 }
 nodes [
  n1 = MaterialFunctionCall(Float2=TextureCoordinate()) {materialFunction=&BreakOutFloat2Components}
  n2 = MaterialFunctionCall(X=@n1.G, Y=@n1.G) {materialFunction=&MakeFloat2}
  n3 = TextureSampleParameter2D(UVs=@n2.Result) {
   parameterName=BodyShapeTexture
   samplerType=SAMPLERTYPE_LinearColor
   texture=/BRCosmetics/Effects/Fort_Effects/Emotes/Emperor/FX/Texture/CurveAtlas_Emote_Emperor_Male_BodyWPO
  }
  Multiply(A=Multiply(A=VertexNormalWS(), B=@n3.R), B=ScalarParameter() {parameterName=WPOStr})
  n4 = TextureSampleParameter2D(UVs=@n2.Result) {
   parameterName=BodyColor0
   samplerType=SAMPLERTYPE_LinearColor
   texture=/BRCosmetics/Effects/Fort_Effects/Emotes/Emperor/FX/Texture/CurveAtlas_Emote_Emperor_Male_BodyColor
  }
  n5 = TextureSampleParameter2D(UVs=@n2.Result) {
   parameterName=BodyColor1
   samplerType=SAMPLERTYPE_LinearColor
   texture=&CurveAtlas_Emote_Emperor_Male_BodyColor
  }
  n6 = LinearInterpolate(
   A=@n4.RGBA
   B=@n5.RGBA
   Alpha=DynamicParameter() {parameterIndex=2, defaultValue=<rgba 0 1 1 1>, paramNames [CurveBlend,Param2,Param3,Param4]}.CurveBlend
  )
  n7 = ComponentMask(None=@n6) {a=true}
  n8 = MaterialFunctionCall(Float3=MaterialFunctionCall() {materialFunction=&BoundingBoxBased_0-1_UVW}.RGB) {materialFunction=&BreakOutFloat3Components}
  n9 = Multiply(A=Add(A=@n8.B, B=ScalarParameter() {parameterName=BottomFade Offset}), B=ScalarParameter() {defaultValue=1, parameterName=BottomFade Width})
  n10 = Multiply(
   A=Add(A=OneMinus(None=@n8.B), B=ScalarParameter() {defaultValue=1, parameterName=TopFade Offset})
   B=ScalarParameter() {defaultValue=1, parameterName=TopFade Width}
  )
  n11 = VectorParameter() {parameterName=RotateAxis, defaultValue=<rgba 1 0 0 1>}
  n12 = MaterialFunctionCall() {materialFunction=&LocalPosition}
  n13 = Add(
   A=RotateAboutAxis(
    NormalizedRotationAxis=Normalize(VectorInput=@n11.RGB)
    RotationAngle=ScalarParameter() {parameterName=Rotato}
    PivotPoint=VectorParameter() {parameterName=RotateCenter, defaultValue=<rgba 0 0 0 1>}.RGB
    Position=@n12."Local Position (Excluding Offsets)"
   )
   B=@n12."Local Position (Excluding Offsets)"
  )
  n14 = MaterialFunctionCall(Float3=@n13) {materialFunction=&BreakOutFloat3Components}
  n15 = Normalize(VectorInput=AppendVector(A=@n14.R, B=@n14.R))
  n16 = Constant2Vector() {r=1}
  n17 = Divide(A=ArccosineFast!(None=DotProduct(A=@n15, B=@n16)), B=ScalarParameter() {defaultValue=3.1416, parameterName=Pi})
  n18 = MaterialFunctionCall(Float2=@n15) {materialFunction=&BreakOutFloat2Components}
  n19 = LinearInterpolate(A=@n17, B=Add(A=OneMinus(None=@n17)), Alpha=Saturate(None=Ceil(None=@n18.G)))
  n20 = MaterialFunctionCall() {materialFunction=&ObjectLocalBounds}
  n21 = Divide(A=Subtract(A=@n13, B=@n20.Local Bounds Minimum), B=@n20.Local Bounds Size)
  n22 = VectorParameter() {
   group=Emissive
   parameterName=Emissive TilingMethod
   sortPriority=3
   channelNames {a=Use Dynamic Offset, b=Rotation, g=V Clamp, r=U Clamp}
   defaultValue=<rgba 0 1 0 0>
  }
  n23 = MaterialFunctionCall(
   UVs=MaterialFunctionCall(
    X=Multiply(A=@n19) {constA=0.5, constB=0.5}
    Y=MaterialFunctionCall(Float3=@n21) {materialFunction=&BreakOutFloat3Components}.B
   ) {materialFunction=&MakeFloat2}.Result
   "Rotation Angle (0-1)"=@n22.Rotation
  ) {materialFunction=&CustomRotator}
  n24 = VectorParameter() {group=Emissive, parameterName=Emissive UV, sortPriority=2, defaultValue=<rgba 1 1 0 0>}
  n25 = DynamicParameter() {defaultValue=<rgba 0 1 1 1>, paramNames [Emi V Offset, Emi V Tiling, Emi U OffsetScale, Dissolve]}
  n26 = MaterialFunctionCall(X=@n24.R, Y=Multiply(A=@n24.G, B=@n25.Emi V Tiling)) {materialFunction=&MakeFloat2}
  n27 = MaterialFunctionCall(X=Multiply(A=@n24.B, B=@n25.Emi U OffsetScale), Y=LinearInterpolate(A=@n24.A, B=@n25.Emi V Offset, Alpha=@n22.Use Dynamic Offset)) {materialFunction=&MakeFloat2}
  n28 = DynamicParameter() {parameterIndex=3, defaultValue=<rgba 0 0 1 1>, paramNames [UVTilingRandom, TimeOffsetRandom, SplineWidthBase, SplineWidthTip]}
  n29 = Multiply(A=@n27.Result, B=Add(A=Time(), B=@n28.TimeOffsetRandom))
  n30 = MaterialFunctionCall(X=@n22.U Clamp, Y=@n22.U Clamp) {materialFunction=&MakeFloat2}
  n31 = Add(B=Add(A=Multiply(A=@n23.Rotated Values, B=@n26.Result), B=LinearInterpolate(A=@n29, B=@n27.Result, Alpha=@n30.Result)))
  n32 = TextureObjectParameter() {group=Emissive, parameterName=Emissive Tex, sortPriority=0}
  n33 = Multiply(
   A=MaterialFunctionCall!(TexCoord=@n31, ClampChannel=@n30.Result) {materialFunction=/Game/Effects/Fort_Effects/MaterialFunctions/MF_FX_TilingMethod}.Result
   B=ChannelMaskParameter(None=TextureSample(UVs=@n31, Tex=@n32).RGBA) {group=Emissive, parameterName=Emissive Channel, sortPriority=1}.R
  )
  n34 = ScalarParameter() {parameterName=BeforeRefineOffset}
  n35 = VectorParameter() {group=Emissive, parameterName=Emissive Refine, sortPriority=4, defaultValue=<rgba 1 1 1 0.5>}
  n36 = Saturate(None=MaterialFunctionCall(Texture=Add(A=@n33, B=@n34), RGB=@n35.RGB, A=@n35.RGB) {materialFunction=&MF_FX_Refine}.Result)
  n37 = ScalarParameter() {parameterName=BodyMaskSharpness}
  n38 = ScalarParameter() {defaultValue=1, parameterName=RegularFresExp}
  n39 = ScalarParameter() {defaultValue=22, parameterName=TBFresExp}
  n40 = Transform(None=Constant3Vector() {constant=<rgba 0 0 1 0>}) {transformSourceType=TRANSFORMSOURCE_Local}
  n41 = DotProduct(A=CameraVectorWS(), B=@n40)
  n42 = Saturate(None=Power(Base=Abs(None=@n41), Exp=ScalarParameter() {defaultValue=3, parameterName=ClosePower}))
  n43 = PixelNormalWS()
  n44 = Constant3Vector() {constant=<rgba 0 0 -1 0>}
  n45 = ScalarParameter() {defaultValue=1.26, parameterName=ScaleNormalChange}
  n46 = TextureSampleParameter2D(UVs=Add(A=@n2.Result, B=ScalarParameter() {defaultValue=0.045, parameterName=StepHeight}) {constB=0.01}) {
   parameterName=BodyShapeTexture
   samplerType=SAMPLERTYPE_LinearColor
   texture=&CurveAtlas_Emote_Emperor_Male_BodyWPO
  }
  n47 = MaterialFunctionCall(
   1=Multiply(A=Transform(None=@n44) {transformSourceType=TRANSFORMSOURCE_Local}, B=@n45)
   2=@n43
   3=Multiply(A=@n40, B=@n45)
   A=Add(A=Subtract(A=@n3.R, B=@n46.R)) {constB=0.5}
  ) {materialFunction=&Lerp_Multiple_Float3}
  n48 = OneMinus(None=Fresnel(
    ExponentIn=LinearInterpolate(A=@n38, B=@n39, Alpha=@n42)
    BaseReflectFractionIn=Constant()
    Normal=LinearInterpolate(A=@n43, B=@n47.Lerp 3 Inputs, Alpha=@n42)
   ))
  n49 = ScalarParameter() {parameterName=FresnelFinalPower}
  n50 = Multiply(A=Multiply(A=Saturate(None=@n9), B=Saturate(None=@n10)), B=Multiply(A=Multiply(A=@n36, B=@n37), B=Power(Base=@n48, Exp=@n49)))
  n51 = ParticleColor()
  Multiply(A=DepthFade(Opacity=Multiply(A=@n7, B=@n50), FadeDistance=ScalarParameter() {parameterName=DepthFade}), B=@n51.A)
  Multiply(A=ComponentMask(None=@n6) {b=true, g=true, r=true}, B=@n51.RGB)
 ]
}
