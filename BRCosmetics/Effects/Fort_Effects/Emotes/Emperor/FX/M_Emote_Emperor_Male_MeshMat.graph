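# Translucent, lit effect material for the male Emperor emote, drawing a desaturatable body texture on skeletal or Niagara particle meshes with optional particle-color tint.
# Rotated local position yields arccosine-and-height UVs that pan a channel-masked texture, adjusted by dynamic parameters.
# Despite its Emissive naming, that mask drives opacity, alongside inverted Fresnel, depth fade, bounding-box top/bottom fades and particle alpha.
# A scalar scales a vertex-normal offset.
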Material /BRCosmetics/Effects/Fort_Effects/Emotes/Emperor/FX/M_Emote_Emperor_Male_MeshMat
graph {
 properties {
  bUsedWithMeshDeformer=true
  bUsedWithNiagaraMeshParticles=true
  bUsedWithSkeletalMesh=true
  blendMode=BLEND_Translucent
  pixelDepthOffsetMode=PDOM_Legacy
  translucencyLightingMode=TLM_VolumetricDirectional
  usedShadingModels=Default Lit
 }
 nodes [
  n1 = VectorParameter() {parameterName=RotateAxis, defaultValue=<rgba 1 0 0 1>}
  n2 = MaterialFunctionCall() {materialFunction=&LocalPosition}
  n3 = Add(
   A=RotateAboutAxis(
    NormalizedRotationAxis=Normalize(VectorInput=@n1.RGB)
    RotationAngle=ScalarParameter() {parameterName=Rotator}
    PivotPoint=VectorParameter() {parameterName=RotateCenter, defaultValue=<rgba 0 0 0 1>}.RGB
    Position=@n2."Local Position (Excluding Offsets)"
   )
   B=@n2."Local Position (Excluding Offsets)"
  )
  n4 = MaterialFunctionCall(Float3=@n3) {materialFunction=&BreakOutFloat3Components}
  n5 = Normalize(VectorInput=AppendVector(A=@n4.R, B=@n4.R))
  n6 = Constant2Vector() {r=1}
  n7 = Divide(A=ArccosineFast!(None=DotProduct(A=@n5, B=@n6)), B=ScalarParameter() {defaultValue=3.1416, parameterName=Pi})
  n8 = MaterialFunctionCall(Float2=@n5) {materialFunction=&BreakOutFloat2Components}
  n9 = LinearInterpolate(A=@n7, B=Add(A=OneMinus(None=@n7)), Alpha=Saturate(None=Ceil(None=@n8.G)))
  n10 = MaterialFunctionCall() {materialFunction=&ObjectLocalBounds}
  n11 = Divide(A=Subtract(A=@n3, B=@n10.Local Bounds Minimum), B=@n10.Local Bounds Size)
  n12 = VectorParameter() {
   group=Emissive
   parameterName=Emissive TilingMethod
   sortPriority=3
   channelNames {a=Use Dynamic Offset, b=Rotation, g=V Clamp, r=U Clamp}
   defaultValue=<rgba 0 1 0 0>
  }
  n13 = MaterialFunctionCall(
   UVs=MaterialFunctionCall(
    X=Multiply(A=@n9) {constA=0.5, constB=0.5}
    Y=MaterialFunctionCall(Float3=@n11) {materialFunction=&BreakOutFloat3Components}.B
   ) {materialFunction=&MakeFloat2}.Result
   "Rotation Angle (0-1)"=@n12.Rotation
  ) {materialFunction=&CustomRotator}
  n14 = VectorParameter() {group=Emissive, parameterName=Emissive UV, sortPriority=2, defaultValue=<rgba 1 1 0 0>}
  n15 = DynamicParameter() {defaultValue=<rgba 0 1 1 1>, paramNames [Emi V Offset, Emi V Tiling, Emi U OffsetScale, Dissolve]}
  n16 = MaterialFunctionCall(X=@n14.R, Y=Multiply(A=@n14.G, B=@n15.Emi V Tiling)) {materialFunction=&MakeFloat2}
  n17 = MaterialFunctionCall(
   X=Multiply(
    A=@n14.B
    B=DynamicParameter() {defaultValue=<rgba 0 1 1 1>, paramNames [Emi V Offset, Emi V Tiling, Emi U OffsetScale, Dissolve]}.Emi U OffsetScale
   )
   Y=LinearInterpolate(A=@n14.A, B=@n15.Emi V Offset, Alpha=@n12.Use Dynamic Offset)
  ) {materialFunction=&MakeFloat2}
  n18 = DynamicParameter() {parameterIndex=3, defaultValue=<rgba 0 0 1 1>, paramNames [UVTilingRandom, TimeOffsetRandom, SplineWidthBase, SplineWidthTip]}
  n19 = Multiply(A=@n17.Result, B=Add(A=Time(), B=@n18.TimeOffsetRandom))
  n20 = MaterialFunctionCall(X=@n12.U Clamp, Y=@n12.U Clamp) {materialFunction=&MakeFloat2}
  n21 = Add(B=Add(A=Multiply(A=@n13.Rotated Values, B=@n16.Result), B=LinearInterpolate(A=@n19, B=@n17.Result, Alpha=@n20.Result)))
  n22 = MaterialFunctionCall!(TexCoord=@n21, ClampChannel=@n20.Result) {materialFunction=/Game/Effects/Fort_Effects/MaterialFunctions/MF_FX_TilingMethod}
  n23 = ChannelMaskParameter(None=TextureSample(UVs=@n21, Tex=TextureObjectParameter() {group=Emissive, parameterName=Emissive Tex, sortPriority=0}).RGBA) {group=Emissive, parameterName=Emissive Channel, sortPriority=1}
  n24 = VectorParameter() {group=Emissive, parameterName=Emissive Refine, sortPriority=4, defaultValue=<rgba 1 1 1 0.5>}
  n25 = Add(
   A=MaterialFunctionCall(Texture=Multiply(A=@n22.Result, B=@n23.R), RGB=@n24.RGB, A=@n24.RGB) {materialFunction=&MF_FX_Refine}.Result
   B=ScalarParameter() {defaultValue=1, group=Emissive, parameterName="Emissive MaskADD ", sortPriority=5}
  )
  n26 = ScalarParameter() {parameterName=Exponent}
  n27 = Constant()
  n28 = Multiply(
   A=Multiply(A=ScalarParameter() {parameterName=BaseOpacity}, B=Saturate(None=@n25))
   B=OneMinus(None=Fresnel(ExponentIn=@n26, BaseReflectFractionIn=@n27))
  )
  n29 = ScalarParameter() {parameterName=DepthFade}
  n30 = MaterialFunctionCall(Float3=MaterialFunctionCall() {materialFunction=&BoundingBoxBased_0-1_UVW}.RGB) {materialFunction=&BreakOutFloat3Components}
  n31 = ScalarParameter() {parameterName=BottomFade Offset}
  n32 = Saturate(None=Multiply(A=Add(A=@n30.B, B=@n31), B=ScalarParameter() {defaultValue=1, parameterName=BottomFade Width}))
  n33 = OneMinus(None=@n30.B)
  n34 = ScalarParameter() {defaultValue=1, parameterName=TopFade Offset}
  n35 = Saturate(None=Multiply(A=Add(A=@n33, B=@n34), B=ScalarParameter() {defaultValue=1, parameterName=TopFade Width}))
  n36 = ParticleColor()
  Multiply(A=Multiply(A=DepthFade(Opacity=@n28, FadeDistance=@n29), B=Multiply(A=@n32, B=@n35)), B=@n36.A)
  Multiply(A=VertexNormalWS(), B=ScalarParameter() {parameterName=VertexOffset})
  n37 = Multiply(
   A=Desaturation(
    None=TextureSampleParameter2D() {parameterName=Param, texture=/BRCosmetics/Effects/Fort_Effects/Emotes/Emperor/FX/Texture/T_M_MED_Emperor_Body_D2}.RGB
    Fraction=ScalarParameter() {parameterName=Desaturation}
   )
   B=ScalarParameter() {parameterName=ScaleColor}
  )
  StaticSwitchParameter(True=Multiply(A=@n37, B=@n36.RGB), False=@n37) {parameterName=CanConvertToDark}
 ]
}
